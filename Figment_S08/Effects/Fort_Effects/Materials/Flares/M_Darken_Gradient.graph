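# Unlit translucent particle material for sprites and mesh particles that makes a soft radial gradient, named as a darkening flare effect.
# A centered UV sphere mask is raised to a dynamic-parameter Falloff exponent, so each particle system can shape the gradient.
# The result is multiplied by particle alpha and a depth fade fed by a near-camera fade, softening geometry intersections and close-up views.
# No color input feeds the graph.
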Material /Figment_S08/Effects/Fort_Effects/Materials/Flares/M_Darken_Gradient
graph {
 properties {
  bUsedWithMeshParticles=true
  bUsedWithNiagaraSprites=true
  bUsedWithParticleSprites=true
  blendMode=BLEND_Translucent
  pixelDepthOffsetMode=PDOM_Legacy
  shadingModel=MSM_Unlit
  usedShadingModels=Unlit
 }
 nodes [
  n1 = TextureCoordinate()
  n2 = Constant() {r=0.5}
  n3 = MaterialFunctionCall(Distance=ScalarParameter() {defaultValue=250, parameterName=NearCameraFadeDist}) {materialFunction=&Near_Camera_Fade}
  n4 = ScalarParameter() {defaultValue=50, parameterName=DepthFade}
  Multiply(
   A=Power(
    Base=SphereMask(A=@n1, B=@n2, Radius=@n2) {hardnessPercent=0}
    Exp=DynamicParameter() {defaultValue=<rgba 3 1 1 1>, paramNames [Falloff,Param2,Param3,Param4]}.Falloff
   ) {constExponent=3}
   B=Multiply(A=ParticleColor().A, B=DepthFade(Opacity=@n3.Result, FadeDistance=@n4))
  )
 ]
}
